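# Masked unlit mesh-particle material for shapes in the Street boost effect, with stylized patterned shading.
# Procedural stripes and dots, blended and tiled by dynamic parameters, darken surfaces facing away from a fixed Z axis, while those facing it stay solid.
# The pattern multiplies particle color, desaturated per particle, then feeds MF_FX_EmissiveAdjustments.
# An IsOutline static switch turns emissive black and selects a TwoSidedSign term instead of constant one.
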Material /VehicleCosmetics/Boosts/Boost_Street/Materials/M_Boost_Street_Shape
graph {
 properties {
  bUsedWithNiagaraMeshParticles=true
  blendMode=BLEND_Masked
  pixelDepthOffsetMode=PDOM_Legacy
  shadingModel=MSM_Unlit
  usedShadingModels=Unlit
 }
 nodes [
  n1 = TextureCoordinate() {uTiling=64, vTiling=0}
  MaterialExpressionMeshEmitterDynamicParameter_1 = DynamicParameter() {paramNames [shadowType,saturation,tile,Param4]}
  n2 = AppendVector(A=@MaterialExpressionMeshEmitterDynamicParameter_1.tile, B=Constant())
  n3 = Subtract(A=Frac(None=Multiply(A=@n1, B=@n2)), B=Constant() {r=0.5})
  n4 = DotProduct(A=@n3, B=@n3)
  n5 = Constant() {r=0.275}
  MaterialExpressionConstantClamp_6 = Saturate(None=Multiply(A=Subtract(A=@n4, B=@n5), B=Constant() {r=512}))
  n6 = TextureCoordinate() {uTiling=32, vTiling=32}
  n7 = Subtract(A=Frac(None=Multiply(A=@n6, B=@MaterialExpressionMeshEmitterDynamicParameter_1.tile)), B=Constant() {r=0.5})
  n8 = DotProduct(A=@n7, B=@n7)
  n9 = Constant() {r=0.15}
  MaterialExpressionConstantClamp_7 = Saturate(None=Multiply(A=Subtract(A=@n8, B=@n9), B=Constant() {r=512}))
  n10 = Transform(None=Constant3Vector() {constant=<rgba 0 0 1 0>})
  n11 = Constant3Vector() {constant=<rgba 0 0 1 0>}
  MaterialExpressionConstantClamp_5 = Saturate(None=Ceil(None=DotProduct(A=@n10, B=@n11)))
  MaterialExpressionConstantClamp_4 = Saturate(None=LinearInterpolate(
    A=LinearInterpolate(
     A=@MaterialExpressionConstantClamp_6
     B=@MaterialExpressionConstantClamp_7
     Alpha=@MaterialExpressionMeshEmitterDynamicParameter_1.shadowType
    )
    B=Constant() {r=1}
    Alpha=@MaterialExpressionConstantClamp_5
   ))
  MaterialExpressionMeshEmitterVertexColor_1 = ParticleColor()
  n12 = Desaturation(None=@MaterialExpressionMeshEmitterVertexColor_1.RGB, Fraction=@MaterialExpressionMeshEmitterDynamicParameter_1.saturation)
  n13 = StaticBoolParameter() {parameterName=IsOutline}
  MaterialFunctionCall(Emissive=StaticSwitch(True=Constant(), False=Multiply(A=@MaterialExpressionConstantClamp_4, B=@n12), Value=@n13)) {materialFunction=&MF_FX_EmissiveAdjustments}
  n14 = TwoSidedSign()
  OneMinus(None=@n14)
  StaticSwitch(True=@n14, False=Constant() {r=1}, Value=@n13)
 ]
}
