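# Unlit translucent Niagara sprite material for the outline of the street smoke effect in the Street boost cosmetic.
# The outline texture's green channel minus particle alpha is steeply multiplied and saturated, giving a hard-edged erosion mask driven by particle alpha.
# A default (black) constant color feeds the shared FX emissive adjustments function.
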Material /VehicleCosmetics/Boosts/Boost_Street/Materials/M_Boost_Street_StreetSmoke_Outline
graph {
 properties {bUsedWithNiagaraSprites=true, blendMode=BLEND_Translucent, pixelDepthOffsetMode=PDOM_Legacy, shadingModel=MSM_Unlit, usedShadingModels=Unlit}
 nodes [
  n1 = TextureSample() {texture=/VehicleCosmetics/Boosts/Boost_Street/Textures/T_VCVFX_Boost_Street_StreetSmokeOutline_N2}
  MaterialExpressionVertexColor_2 = ParticleColor()
  MaterialExpressionConstantClamp_18 = Saturate(None=Multiply(A=Subtract(A=@n1.G, B=@MaterialExpressionVertexColor_2.A), B=Constant() {r=64}))
  MaterialFunctionCall(Emissive=Constant3Vector()) {materialFunction=&MF_FX_EmissiveAdjustments}
 ]
}
