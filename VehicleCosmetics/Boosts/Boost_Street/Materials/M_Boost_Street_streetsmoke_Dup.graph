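# Translucent unlit Niagara sprite material for the Street boost's stylized smoke, tinted by particle color.
# Steeply thresholded normal-map up-facing lighting, broken up by a repeating particle-relative world-space dot pattern, picks between particle color and a dynamic-parameter hue-shifted, darkened shade.
# A steep threshold on texture alpha cuts the silhouette.
# The Fill switch substitutes a procedural hard-edged disc; FullOpacity bypasses the Alpha_Test fade from black.
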Material /VehicleCosmetics/Boosts/Boost_Street/Materials/M_Boost_Street_streetsmoke_Dup
graph {
 properties {
  bUsedWithNiagaraSprites=true
  blendMode=BLEND_Translucent
  opacityMaskClipValue=0.5
  pixelDepthOffsetMode=PDOM_Legacy
  shadingModel=MSM_Unlit
  twoSided=true
  usedShadingModels=Unlit
 }
 nodes [
  n1 = ParticleColor()
  n2 = Constant()
  n3 = ScalarParameter() {defaultValue=1, parameterName=Alpha_Test}
  n4 = StaticSwitchParameter(True=@n1.RGB, False=LinearInterpolate(A=@n2, B=@n1.RGB, Alpha=@n3)) {parameterName=FullOpacity}
  n5 = Add(A=TextureCoordinate(), B=Constant() {r=-0.5})
  n6 = DotProduct(A=@n5, B=@n5)
  n7 = ConstantBiasScale(None=@n6) {bias=-0.05, scale=512}
  n8 = ConstantBiasScale(None=OneMinus(None=@n6)) {bias=-0.8, scale=512}
  n9 = OneMinus(None=Multiply(A=Saturate(None=@n7), B=@n8))
  n10 = DynamicParameter() {paramNames [hueshift,shadow,painted,Param4]}
  n11 = MaterialFunctionCall(Hue Shift Percentage=ComponentMask(None=@n10.hueshift) {r=true}, Texture=@n1.RGB) {materialFunction=&HueShift}
  n12 = Transform(None=TextureSample() {samplerType=SAMPLERTYPE_Normal, texture=/VehicleCosmetics/Boosts/Boost_Street/Textures/T_VCVFX_Boost_Street_StreetSmoke_N2}.RGB)
  n13 = Constant3Vector() {constant=<rgba 0 0 1 0>}
  n14 = DynamicParameter() {parameterIndex=1, paramNames [mul,div,painted,Param4]}
  n15 = Divide(A=Subtract(A=WorldPosition().XYZ, B=ParticlePositionWS()), B=@n14.div)
  n16 = Subtract(A=Frac(None=Multiply(A=@n15, B=@n14.mul)), B=Constant() {r=0.5})
  n17 = Subtract(A=Saturate(None=DotProduct(A=@n12, B=@n13)), B=DotProduct(A=@n16, B=@n16))
  n18 = Constant() {r=512}
  n19 = LinearInterpolate(
   A=Multiply(A=LinearInterpolate(A=@n11.Result, B=@n1.RGB, Alpha=@n10.painted), B=@n10.shadow)
   B=@n1.RGB
   Alpha=Saturate(None=Multiply(A=@n17, B=@n18))
  )
  n20 = LinearInterpolate(A=@n2, B=@n19, Alpha=@n3)
  MaterialFunctionCall(Emissive=StaticSwitchParameter(True=Multiply(A=@n4, B=@n9), False=StaticSwitchParameter(True=@n19, False=@n20) {parameterName=FullOpacity}) {parameterName=Fill}) {materialFunction=&MF_FX_EmissiveAdjustments}
  n21 = Subtract(
   A=TextureSample() {samplerType=SAMPLERTYPE_LinearColor, texture=/VehicleCosmetics/Boosts/Boost_Street/Textures/T_VCVFX_Boost_Street_StreetSmoke_N}.A
   B=Constant() {r=0.6}
  )
  n22 = Constant() {r=20}
  StaticSwitchParameter(True=Saturate(None=@n8), False=Saturate(None=Multiply(A=@n21, B=@n22))) {parameterName=Fill}
 ]
}
